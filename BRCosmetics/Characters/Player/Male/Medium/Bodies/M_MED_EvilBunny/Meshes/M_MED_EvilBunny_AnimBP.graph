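# Secondary-motion AnimBP for the EvilBunny player body cosmetic, driving collar, hooks and ear-ball bones over the pose copied from the parent mesh.
# Each update maps nose-to-collar socket distance and head/clavicle rotation deltas to alphas.
# These alphas blend additive collar bone offsets after head-driven Flap nodes, keeping the collar responsive to pose.
# Weighted constraints anchor hooks before AnimDynamics simulates them and the ear balls; tilt alphas are computed but unused.
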AnimBlueprint /BRCosmetics/Characters/Player/Male/Medium/Bodies/M_MED_EvilBunny/Meshes/M_MED_EvilBunny_AnimBP
target_skeleton=&/M_MED_EvilBunny_Skeleton
graphs [
 AnimGraph {nodes [
   n1 = Get clavicle_up_l_alpha() -> (clavicle_up_l_alpha:double)
   n2 = Flap Node - Bone: dyn_collar_01!(
    ComponentPose:Component Space Pose Link Structure=Local To Component(LocalPose=Copy Pose From Mesh() {bUseAttachedParent=true})
    Alpha:float=1
   ) -> (Pose:Component Space Pose Link Structure) {
    is=title FortAnimGraphNode_Flap
    bUseClamp=true
    targetClampMax=2.5
    targetComponent=RotationZ
    alphaBoolBlend {blendOption=Linear}
    sourceDrives [{
      bUseRange=false
      drivingCurve=null
      multiplier=5
      rangeMax=0
      rangeMin=0
      remappedMax=0
      remappedMin=0
      sourceComponent=RotationZ
      useQuaternion=true
      sourceBone {boneName=head}
     }]
    targetBone {boneName=dyn_collar_01}
   }
   n3 = Flap Node - Bone: dyn_collar_03_l!(ComponentPose:Component Space Pose Link Structure=@n2, Alpha:float=1) -> (Pose:Component Space Pose Link Structure) {
    is=title FortAnimGraphNode_Flap
    bUseClamp=true
    targetClampMax=3.14
    targetComponent=RotationZ
    alphaBoolBlend {blendOption=Linear}
    sourceDrives [{
      bUseRange=false
      drivingCurve=null
      multiplier=1
      rangeMax=0
      rangeMin=0
      remappedMax=0
      remappedMin=0
      sourceComponent=RotationZ
      useQuaternion=true
      sourceBone {boneName=head}
     }]
    targetBone {boneName=dyn_collar_03_l}
   }
   n4 = Flap Node - Bone: dyn_collar_03_r!(ComponentPose:Component Space Pose Link Structure=@n3, Alpha:float=1) -> (Pose:Component Space Pose Link Structure) {
    is=title FortAnimGraphNode_Flap
    bUseClamp=true
    targetClampMax=3.14
    targetComponent=RotationZ
    alphaBoolBlend {blendOption=Linear}
    sourceDrives [{
      bUseRange=false
      drivingCurve=null
      multiplier=1
      rangeMax=0
      rangeMin=0
      remappedMax=0
      remappedMin=0
      sourceComponent=RotationZ
      useQuaternion=true
      sourceBone {boneName=head}
     }]
    targetBone {boneName=dyn_collar_03_r}
   }
   n5 = "Transform (Modify) Bone - Bone: dyn_collar_01_l"(ComponentPose=@n4, Alpha=Get collar_01_l() -> (collar_01_l:double), Translation="0,0,0", Rotation="20,0,0", Scale="1,1,1") {
    rotationMode=BMM_Additive
    rotationSpace=BCS_BoneSpace
    alphaBoolBlend {blendOption=Linear}
    boneToModify {boneName=dyn_collar_01_l}
    rotation {pitch=20}
   }
   n6 = "Transform (Modify) Bone - Bone: dyn_collar_01_r"(ComponentPose=@n5, Alpha=Get collar_01_r() -> (collar_01_r:double), Translation="0,0,0", Rotation="20,0,0", Scale="1,1,1") {
    rotationMode=BMM_Additive
    rotationSpace=BCS_BoneSpace
    alphaBoolBlend {blendOption=Linear}
    boneToModify {boneName=dyn_collar_01_r}
    rotation {pitch=20}
   }
   n7 = "Transform (Modify) Bone - Bone: dyn_collar_01_l"(ComponentPose=@n6, Alpha=Get left_max_down() -> (left_max_down:double), Translation="0,0,0", Rotation="60,0,0", Scale="1,1,1") {
    rotationMode=BMM_Additive
    rotationSpace=BCS_BoneSpace
    alphaBoolBlend {blendOption=Linear}
    boneToModify {boneName=dyn_collar_01_l}
    rotation {pitch=60}
   }
   n8 = Get left_max_alpha() -> (left_max_alpha:double)
   n9 = "Transform (Modify) Bone - Bone: dyn_collar_02_l"(ComponentPose=@n7, Alpha=@n8, Translation="0,0,0", Rotation="60,0,0", Scale="1,1,1") {
    rotationMode=BMM_Additive
    rotationSpace=BCS_BoneSpace
    alphaBoolBlend {blendOption=Linear}
    boneToModify {boneName=dyn_collar_02_l}
    rotation {pitch=60}
   }
   n10 = "Transform (Modify) Bone - Bone: dyn_collar_03_l"(ComponentPose=@n9, Alpha=@n8, Translation="0,0,0", Rotation="40,0,0", Scale="1,1,1") {
    rotationMode=BMM_Additive
    rotationSpace=BCS_BoneSpace
    alphaBoolBlend {blendOption=Linear}
    boneToModify {boneName=dyn_collar_03_l}
    rotation {pitch=40}
   }
   n11 = "Transform (Modify) Bone - Bone: dyn_collar_01_r"(ComponentPose=@n10, Alpha=Get right_max_down() -> (right_max_down:double), Translation="0,0,0", Rotation="40,0,0", Scale="1,1,1") {
    rotationMode=BMM_Additive
    rotationSpace=BCS_BoneSpace
    alphaBoolBlend {blendOption=Linear}
    boneToModify {boneName=dyn_collar_01_r}
    rotation {pitch=40}
   }
   n12 = Get right_max_alpha() -> (right_max_alpha:double)
   n13 = "Transform (Modify) Bone - Bone: dyn_collar_02_r"(ComponentPose=@n11, Alpha=@n12, Translation="0,0,0", Rotation="40,0,0", Scale="1,1,1") {
    rotationMode=BMM_Additive
    rotationSpace=BCS_BoneSpace
    alphaBoolBlend {blendOption=Linear}
    boneToModify {boneName=dyn_collar_02_r}
    rotation {pitch=40}
   }
   n14 = "Transform (Modify) Bone - Bone: dyn_collar_03_r"(ComponentPose=@n13, Alpha=@n12, Translation="0,0,0", Rotation="40,0,0", Scale="1,1,1") {
    rotationMode=BMM_Additive
    rotationSpace=BCS_BoneSpace
    alphaBoolBlend {blendOption=Linear}
    boneToModify {boneName=dyn_collar_03_r}
    rotation {pitch=40}
   }
   n15 = "Constraint - dyn_upper_hook (PPPPPPPP)"!(
    ComponentPose=@n14
    Alpha=1
    ConstraintWeights_0=0.0689
    ConstraintWeights_1:float=0.1425
    ConstraintWeights_2:float=0.2812
    ConstraintWeights_3:float=0.2976
    ConstraintWeights_4:float=0.1551
    ConstraintWeights_5:float=0.0011
    ConstraintWeights_6:float=0.0001
    ConstraintWeights_7:float=0.0533
   ) -> (Pose) {
    alphaBoolBlend {blendOption=Linear}
    boneToModify {boneName=dyn_upper_hook}
    constraintSetup [
     {targetBone {boneName=clavicle_l}}
     {targetBone {boneName=spine_01}}
     {targetBone {boneName=spine_02}}
     {targetBone {boneName=spine_03}}
     {targetBone {boneName=spine_04}}
     {targetBone {boneName=spine_05}}
     {targetBone {boneName=clavicle_r}}
     {targetBone {boneName=pelvis}}
    ]
    constraintWeights [1,1,1,1,1,1,1,1]
   }
   n16 = "Constraint - dyn_lower_hook_chain (PPPPPPP)"!(
    ComponentPose=@n15
    Alpha=1
    ConstraintWeights_0=0.0111
    ConstraintWeights_1:float=0.1755
    ConstraintWeights_2:float=0.2563
    ConstraintWeights_3:float=0.2179
    ConstraintWeights_4:float=0.0675
    ConstraintWeights_5:float=0
    ConstraintWeights_6:float=0.2717
   ) -> (Pose) {
    alphaBoolBlend {blendOption=Linear}
    boneToModify {boneName=dyn_lower_hook_chain}
    constraintSetup [
     {targetBone {boneName=clavicle_l}}
     {targetBone {boneName=spine_01}}
     {targetBone {boneName=spine_02}}
     {targetBone {boneName=spine_03}}
     {targetBone {boneName=spine_04}}
     {targetBone {boneName=clavicle_r}}
     {targetBone {boneName=pelvis}}
    ]
    constraintWeights [1,1,1,1,1,1,1]
   }
   n17 = Anim Dynamics - Bone: dyn_upper_hook(ComponentPose=@n16, Alpha=1, ExternalForce="0,0,0") {
    angularDampingOverride=1
    angularSpringConstant=1500
    bAngularSpring=true
    bGravityOverrideInSimSpace=true
    bOverrideAngularDamping=true
    gravityScale=0.35
    simulationSpace=RootRelative
    alphaBoolBlend {blendOption=Linear}
    boundBone {boneName=dyn_upper_hook}
    physicsBodyDefinitions [{
      sphereCollisionRadius=0
      boundBone {boneName=dyn_upper_hook}
      boxExtents=<6 15 15>
      constraintSetup {angularLimitsMax=<0 0 40>, angularLimitsMin=<0 -60 -40>, angularTarget=<0 0 0>}
      localJointOffset=<6 0 0>
     }]
   }
   n18 = Anim Dynamics - Bone: dyn_lower_hook_chain(ComponentPose=@n17, Alpha=1, ExternalForce="0,0,0") {
    angularDampingOverride=0.9999
    bGravityOverrideInSimSpace=true
    bOverrideAngularDamping=true
    gravityScale=0.35
    simulationSpace=RootRelative
    alphaBoolBlend {blendOption=Linear}
    boundBone {boneName=dyn_lower_hook_chain}
    physicsBodyDefinitions [{
      sphereCollisionRadius=0
      boundBone {boneName=dyn_lower_hook_chain}
      boxExtents=<5 10 10>
      constraintSetup {angularLimitsMax=<0 0 0>, angularLimitsMin=<0 0 -40>, angularTarget=<0 0 0>}
      localJointOffset=<5 0 0>
     }]
   }
   n19 = Anim Dynamics - Bone: dyn_lower_hook(ComponentPose=@n18, Alpha=1, ExternalForce="0,0,0") {
    angularDampingOverride=1
    angularSpringConstant=1500
    bAngularSpring=true
    bGravityOverrideInSimSpace=true
    bOverrideAngularDamping=true
    gravityScale=0.35
    simulationSpace=RootRelative
    alphaBoolBlend {blendOption=Linear}
    boundBone {boneName=dyn_lower_hook}
    physicsBodyDefinitions [{
      sphereCollisionRadius=0
      boundBone {boneName=dyn_lower_hook}
      boxExtents=<5 10 10>
      constraintSetup {angularLimitsMax=<0 45 0>, angularLimitsMin=<0 -45 -40>, angularTarget=<0 0 0>}
      localJointOffset=<5 0 0>
     }]
   }
   n20 = "Transform (Modify) Bone - Bone: dyn_collar_02_l"(ComponentPose=@n19, Alpha=@n1, Translation="-1.729,1.471,-4.522", Rotation="-35.000,0.000,-10.000", Scale="1,1,1") {
    rotationMode=BMM_Additive
    rotationSpace=BCS_BoneSpace
    translationMode=BMM_Additive
    translationSpace=BCS_BoneSpace
    alphaBoolBlend {blendOption=Linear}
    boneToModify {boneName=dyn_collar_02_l}
    rotation=<rot -35 0 -10>
    translation=<-1.729 1.471 -4.522>
   }
   n21 = "Transform (Modify) Bone - Bone: dyn_collar_03_l"(ComponentPose=@n20, Alpha=0, Translation="-1.675,0.927,-14.128", Rotation="-20.000,0.000,10.000", Scale="1,1,1") {
    rotationMode=BMM_Additive
    rotationSpace=BCS_BoneSpace
    translationMode=BMM_Additive
    translationSpace=BCS_BoneSpace
    alphaBoolBlend {blendOption=Linear}
    boneToModify {boneName=dyn_collar_03_l}
    rotation {pitch=-20, roll=10}
    translation=<-1.6751 0.927 -14.1284>
   }
   n22 = "Transform (Modify) Bone - Bone: dyn_collar_01"(ComponentPose=@n21, Alpha=@n1, Translation="0,0,0", Rotation="0,0,0", Scale="1,1,1") {
    rotationMode=BMM_Additive
    rotationSpace=BCS_BoneSpace
    translationMode=BMM_Additive
    translationSpace=BCS_BoneSpace
    alphaBoolBlend {blendOption=Linear}
    boneToModify {boneName=dyn_collar_01}
   }
   n23 = Anim Dynamics - Bone: dyn_ball_l_ear_r(ComponentPose=@n22, Alpha=1, ExternalForce="0,0,0") {
    bGravityOverrideInSimSpace=true
    gravityScale=1.5
    simulationSpace=World
    alphaBoolBlend {blendOption=Linear}
    boundBone {boneName=dyn_ball_l_ear_r}
    physicsBodyDefinitions [{
      sphereCollisionRadius=0
      boundBone {boneName=dyn_ball_l_ear_r}
      boxExtents=<20 20 5>
      constraintSetup {twistAxis=AxisZ, angularLimitsMax=<15 0 0>, angularLimitsMin=<-15 0 0>, angularTarget=<0 0 0>}
      localJointOffset=<0 0.025 0>
     }]
   }
   n24 = Anim Dynamics - Bone: dyn_ball_l_ear_l(ComponentPose=@n23, Alpha=1, ExternalForce="0,0,0") {
    bGravityOverrideInSimSpace=true
    gravityScale=1.5
    simulationSpace=World
    alphaBoolBlend {blendOption=Linear}
    boundBone {boneName=dyn_ball_l_ear_l}
    physicsBodyDefinitions [{
      sphereCollisionRadius=0
      boundBone {boneName=dyn_ball_l_ear_l}
      boxExtents=<20 20 5>
      constraintSetup {twistAxis=AxisZ, angularLimitsMax=<15 0 0>, angularLimitsMin=<-15 0 0>, angularTarget=<0 0 0>}
      localJointOffset=<0 0.025 0>
     }]
   }
   n25 = Anim Dynamics - Bone: dyn_ball_r_ear_r(ComponentPose=@n24, Alpha=1, ExternalForce="0,0,0") {
    bGravityOverrideInSimSpace=true
    gravityScale=1.5
    simulationSpace=World
    alphaBoolBlend {blendOption=Linear}
    boundBone {boneName=dyn_ball_r_ear_r}
    physicsBodyDefinitions [{
      sphereCollisionRadius=0
      boundBone {boneName=dyn_ball_r_ear_r}
      boxExtents=<20 20 5>
      constraintSetup {twistAxis=AxisZ, angularLimitsMax=<15 0 0>, angularLimitsMin=<-15 0 0>, angularTarget=<0 0 0>}
      localJointOffset=<0 0.025 0>
     }]
   }
   n26 = Anim Dynamics - Bone: dyn_ball_r_ear_l(ComponentPose=@n25, Alpha=1, ExternalForce="0,0,0") {
    bGravityOverrideInSimSpace=true
    gravityScale=1.5
    simulationSpace=World
    alphaBoolBlend {blendOption=Linear}
    boundBone {boneName=dyn_ball_r_ear_l}
    physicsBodyDefinitions [{
      sphereCollisionRadius=0
      boundBone {boneName=dyn_ball_r_ear_l}
      boxExtents=<20 20 5>
      constraintSetup {twistAxis=AxisZ, angularLimitsMax=<15 0 0>, angularLimitsMin=<-15 0 0>, angularTarget=<0 0 0>}
      localJointOffset=<0 0.025 0>
     }]
   }
   Output Pose(Result=Component To Local(ComponentPose=@n26)) {layerGroup=None}
  ]}
 EventGraph {nodes [
   Event BlueprintUpdateAnimation()
   n1 = Get PartSkelMesh() -> (PartSkelMesh:Skeletal Mesh Component Object Reference)
   > Set collar_01_l(collar_01_l:double=Get Distance Between Two Sockets!(
     Component:Skeletal Mesh Component Object Reference=@n1
     SocketOrBoneNameA:Name=nose_tip
     SocketSpaceA:ERelativeTransformSpace Enum=RTS_World
     SocketOrBoneNameB:Name=dyn_collar_01_l
     SocketSpaceB:ERelativeTransformSpace Enum=RTS_World
     bRemapRange:bool=true
     InRangeMin:float=15
     InRangeMax:float=6
     OutRangeMin:float=0
     OutRangeMax:float=1
    ) -> (ReturnValue:float) {is=title K2Node_CallFunction}) -> (Output_Get:double)
   > Set collar_01_r(collar_01_r:double=Get Distance Between Two Sockets(
     Component=@n1
     SocketOrBoneNameA=nose_tip
     SocketSpaceA=RTS_World
     SocketOrBoneNameB=dyn_collar_01_r
     SocketSpaceB=RTS_World
     bRemapRange=true
     InRangeMin=15
     InRangeMax=6
     OutRangeMin=0
     OutRangeMax=1
    )) -> (Output_Get:double)
   > n2 = Sequence!(execute) -> (then_0, then_1, then_2:Exec, then_3:Exec)
   n3 = Get PartSkelMesh()
   n4 = BreakRotator(InRot=BreakTransform(InTransform=GetDeltaTransformFromRefPose(self=@n3, BoneName=head, BaseName=spine_05)).Rotation)
   Set look_left_alpha(execute=@n2.then_2, look_left_alpha:double=MapRangeClamped(Value=@n4, InRangeA=0, InRangeB=40)) -> (Output_Get:double)
   > Set look_right_alpha(look_right_alpha:double=MapRangeClamped(Value=@n4, InRangeA=0, InRangeB=-20)) -> (Output_Get:double)
   n5 = BreakTransform(InTransform=GetDeltaTransformFromRefPose(self=Get PartSkelMesh(), BoneName=head, BaseName=spine_05))
   Set look_down_alpha(execute=@n2.then_1, look_down_alpha:double=MapRangeClamped(Value=BreakRotator(InRot=@n5.Rotation).Yaw, InRangeA=-10, InRangeB=25)) -> (Output_Get:double)
   n6 = Get PartSkelMesh()
   n7 = BreakRotator(InRot=BreakTransform(InTransform=GetDeltaTransformFromRefPose(self=@n6, BoneName=head, BaseName=spine_05)).Rotation)
   Set tilt_left_alpha(execute=@n2.then_3, tilt_left_alpha:double=MapRangeClamped(Value=@n7.Pitch, InRangeA=0, InRangeB=10)) -> (Output_Get:double)
   > Set tilt_right_alpha(tilt_right_alpha:double=MapRangeClamped(Value=@n7.Pitch, InRangeA=0, InRangeB=-10)) -> (Output_Get:double)
   n8 = "Max (Float)"(A=Get look_left_alpha() -> (look_left_alpha:double), B=0)
   > Set left_max_alpha(left_max_alpha:double=@n8) -> (Output_Get:double)
   > Set left_max_down(left_max_down:double="Max (Float)"(A=@n8, B=Get look_down_alpha() -> (look_down_alpha:double))) -> (Output_Get:double)
   n9 = "Max (Float)"(A=Get look_right_alpha() -> (look_right_alpha:double), B=0)
   > Set right_max_alpha(right_max_alpha:double=@n9) -> (Output_Get:double)
   > Set right_max_down(right_max_down:double="Max (Float)"(A=@n9, B=Get look_down_alpha())) -> (Output_Get:double)
   n10 = BreakTransform(InTransform=GetDeltaTransformFromRefPose(self=Get PartSkelMesh(), BoneName=clavicle_l, BaseName=spine_05))
   Set clavicle_up_l_alpha(execute=@n2.then_0, clavicle_up_l_alpha:double=MapRangeClamped(Value=BreakRotator(InRot=@n10.Rotation).Pitch, InRangeA=10, InRangeB=40)) -> (Output_Get:double)
   Event BlueprintInitializeAnimation()
   > Set PartSkelMesh(PartSkelMesh:Skeletal Mesh Component Object Reference=GetPartSkeletalMeshComponent()) -> (Output_Get:Skeletal Mesh Component Object Reference)
   TryGetPawnOwner()
   Get tilt_left_alpha() -> (tilt_left_alpha:double)
   Get tilt_right_alpha() -> (tilt_right_alpha:double)
  ]}
]
variables [
 clavicle_up_l_alpha {type=number}
 collar_01_l {type=number}
 collar_01_r {type=number}
 left_max_alpha {type=number}
 left_max_down {type=number}
 look_down_alpha {type=number}
 look_left_alpha {type=number}
 look_right_alpha {type=number}
 PartSkelMesh {type=SkeletalMeshComponent}
 right_max_alpha {type=number}
 right_max_down {type=number}
 tilt_left_alpha {type=number}
 tilt_right_alpha {type=number}
]
