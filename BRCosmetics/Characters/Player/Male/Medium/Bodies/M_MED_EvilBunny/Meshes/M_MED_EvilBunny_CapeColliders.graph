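# Physics asset that supplies collision volumes for the EvilBunny medium male player body's cape, judging by its name.
# Capsules cover the clavicles, neck, pelvis and spine chain, with a sphere on the head and extra rotated capsules widening spine_02 and a second one on spine_05.
# Arm, leg and foot bodies hold no shapes, and no constraints exist, so it serves as collision only rather than a ragdoll.
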PhysicsAsset /BRCosmetics/Characters/Player/Male/Medium/Bodies/M_MED_EvilBunny/Meshes/M_MED_EvilBunny_CapeColliders
bodies {
 calf_l []
 calf_r []
 clavicle_l [{shapeType=Capsule, length=7.7636, radius=11.1427, center=<9.9642 -0.7741 -1.5923>, rotation=<rot 75.9352 -0.0001 175.5622>}]
 clavicle_r [{shapeType=Capsule, length=7.7636, radius=11.1427, center=<-9.9642 0.7741 1.5923>, rotation=<rot 75.9351 -0.0001 -4.4378>}]
 foot_l []
 foot_r []
 head [{shapeType=Sphere, radius=18.0625, center=<11.028 -0.174 0>}]
 lowerarm_l []
 lowerarm_r []
 neck_02 [{shapeType=Capsule, length=13.4375, radius=7.3175, center=<-0.2642 -0.3408 0>, rotation=<rot 90 -171.1193 -175.3978>}]
 pelvis [{shapeType=Capsule, length=7.6562, radius=16.7888, center=<-2.6215 -1.3695 0>}]
 spine_01 [{shapeType=Capsule, length=9.4687, radius=13.6638, center=<-2.6215 -4.4364 0>}]
 spine_02 [
  {shapeType=Capsule, length=10.5937, radius=10.945, center=<-2.6215 0.4993 0.0001>}
  {shapeType=Capsule, length=29.875, radius=11.1013, center=<2.5024 0.4992 3.9825>, rotation=<rot 90 0.0049 0.0049>}
  {shapeType=Capsule, length=29.875, radius=11.1013, center=<2.4704 0.4991 -3.9825>, rotation=<rot 90 -0.0396 -0.0396>}
 ]
 spine_03 [{shapeType=Capsule, length=7.4062, radius=12.5075, center=<-1.953 -1.7451 0.0001>}]
 spine_04 [{shapeType=Capsule, length=13.6562, radius=14.6325, center=<-3.8759 -3.2664 0.0002>}]
 spine_05 [
  {shapeType=Capsule, length=16.8437, radius=18.195, center=<-1.7514 0.6584 0>}
  {shapeType=Capsule, length=12.7812, radius=17.0625, center=<6.0313 3.1143 0>}
 ]
 thigh_l []
 thigh_r []
 upperarm_l []
 upperarm_r []
}
constraints []
