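# Physics asset for the Evil Bunny head mesh of the medium male player body, covering the dynamic ears and upper body.
# Each ear is a chain of four box bodies hung from the head, with tight limits at the root and looser swing along the chain.
# Capsules cover head, neck, upper spine, clavicles and arms.
# Head-to-neck rotation is limited, while the other body joints leave swing and twist free.
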PhysicsAsset /BRCosmetics/Characters/Player/Male/Medium/Bodies/M_MED_EvilBunny/Meshes/Parts/M_MED_Evil_Bunny_Head_Physics
bodies {
 clavicle_l [{shapeType=Capsule, length=8.1227, radius=3.4139, center=<2.1906 4.8721 1.4443>, rotation=<rot 0 25.8837 -48.9446>}]
 clavicle_r [{shapeType=Capsule, length=7.2578, radius=3.5706, center=<-2.5745 -3.8451 -1.8693>, rotation=<rot 0 28.4806 -58.1634>}]
 dyn_ear_01_l [{shapeType=Box, extentX=11, extentY=8, extentZ=3.5, center=<5 0 0>}]
 dyn_ear_01_r [{shapeType=Box, extentX=11, extentY=8, extentZ=3.5, center=<5 0 0>}]
 dyn_ear_02_l [{shapeType=Box, extentX=11, extentY=8, extentZ=3.5, center=<7.5 0 0>, rotation=<rot 0 0 -0.7997>}]
 dyn_ear_02_r [{shapeType=Box, extentX=11, extentY=8, extentZ=3.5, center=<7.5 0 0>}]
 dyn_ear_03_l [{shapeType=Box, extentX=11, extentY=8, extentZ=3.5, center=<7.5 0 0>, rotation=<rot 0 0 -0.7997>}]
 dyn_ear_03_r [{shapeType=Box, extentX=11, extentY=8, extentZ=3.5, center=<7.5 0 0>}]
 dyn_ear_04_l [{shapeType=Box, extentX=11, extentY=8, extentZ=3.5, center=<7.5 0 0>, rotation=<rot 0 0 -0.7997>}]
 dyn_ear_04_r [{shapeType=Box, extentX=11, extentY=8, extentZ=3.5, center=<7.5 0 0>}]
 head [{shapeType=Capsule, length=0, radius=15.1818, center=<5.5225 3.9221 0>}]
 lowerarm_l [{shapeType=Capsule, length=9, radius=7, center=<9.0355 -0.0003 0>, rotation=<rot 90 0 0>}]
 lowerarm_r [{shapeType=Capsule, length=16, radius=6, center=<-7.1785 -1.2937 -0.0003>, rotation=<rot 90 -90 -83.5379>}]
 neck_01 [{shapeType=Capsule, length=0, radius=17.2126, center=<1.3562 5.3514 -0.0109>, rotation=<rot 90 0 0>}]
 spine_05 [{shapeType=Capsule, length=15.7533, radius=16.6428, center=<-2.3518 0.155 0>}]
 upperarm_l [{shapeType=Capsule, length=14.25, radius=7, center=<12.2781 0 0>, rotation=<rot 90 0 0>}]
 upperarm_r [{shapeType=Capsule, length=14.25, radius=7, center=<-14.7922 0.5151 0.8741>, rotation=<rot 86.792 9.882 179.9999>}]
}
constraints [
 {bone1Name=head, bone2Name=neck_01, swing1LimitDegrees=45, swing2LimitDegrees=45, twistLimitDegrees=45}
 {bone1Name=dyn_ear_02_r, bone2Name=dyn_ear_01_r, swing1LimitDegrees=45, swing2LimitDegrees=60, twistLimitDegrees=15}
 {bone1Name=dyn_ear_02_l, bone2Name=dyn_ear_01_l, swing1LimitDegrees=45, swing2LimitDegrees=60, twistLimitDegrees=15}
 {bone1Name=dyn_ear_01_l, bone2Name=head, swing1LimitDegrees=20, swing2LimitDegrees=20, twistLimitDegrees=5}
 {bone1Name=dyn_ear_01_r, bone2Name=head, swing1LimitDegrees=20, swing2LimitDegrees=20, twistLimitDegrees=5}
 {bone1Name=dyn_ear_03_l, bone2Name=dyn_ear_02_l, swing1LimitDegrees=45, swing2LimitDegrees=60, twistLimitDegrees=15}
 {bone1Name=dyn_ear_04_l, bone2Name=dyn_ear_03_l, swing1LimitDegrees=45, swing2LimitDegrees=60, twistLimitDegrees=5}
 {bone1Name=dyn_ear_03_r, bone2Name=dyn_ear_02_r, swing1LimitDegrees=45, swing2LimitDegrees=60, twistLimitDegrees=15}
 {bone1Name=dyn_ear_04_r, bone2Name=dyn_ear_03_r, swing1LimitDegrees=45, swing2LimitDegrees=60, twistLimitDegrees=5}
 {
  bone1Name=neck_01
  bone2Name=spine_05
  swing1LimitDegrees=45
  swing1Motion=Free
  swing2LimitDegrees=45
  swing2Motion=Free
  twistLimitDegrees=45
  twistMotion=Free
 }
 {
  bone1Name=clavicle_l
  bone2Name=spine_05
  swing1LimitDegrees=45
  swing1Motion=Free
  swing2LimitDegrees=45
  swing2Motion=Free
  twistLimitDegrees=45
  twistMotion=Free
 }
 {
  bone1Name=clavicle_r
  bone2Name=spine_05
  swing1LimitDegrees=45
  swing1Motion=Free
  swing2LimitDegrees=45
  swing2Motion=Free
  twistLimitDegrees=45
  twistMotion=Free
 }
 {
  bone1Name=upperarm_l
  bone2Name=clavicle_l
  swing1LimitDegrees=45
  swing1Motion=Free
  swing2LimitDegrees=45
  swing2Motion=Free
  twistLimitDegrees=45
  twistMotion=Free
 }
 {
  bone1Name=lowerarm_l
  bone2Name=upperarm_l
  swing1LimitDegrees=45
  swing1Motion=Free
  swing2LimitDegrees=45
  swing2Motion=Free
  twistLimitDegrees=45
  twistMotion=Free
 }
 {
  bone1Name=lowerarm_r
  bone2Name=upperarm_r
  swing1LimitDegrees=45
  swing1Motion=Free
  swing2LimitDegrees=45
  swing2Motion=Free
  twistLimitDegrees=45
  twistMotion=Free
 }
 {
  bone1Name=upperarm_r
  bone2Name=clavicle_r
  swing1LimitDegrees=45
  swing1Motion=Free
  swing2LimitDegrees=45
  swing2Motion=Free
  twistLimitDegrees=45
  twistMotion=Free
 }
]
