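# Animation Blueprint for the Evil Bunny head cosmetic part, copying the owning mesh's pose and adding secondary ear motion.
# With physics enabled, Anim Dynamics on both ear bases feeds a RigidBody node; targeting switches off the right ear's dynamics.
# When rigid bodies are disabled or LOD exceeds 1, additive ear rotation driven by head tilt feeds trail and spring controllers.
# Trail strength fades as forward pawn speed rises.
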AnimBlueprint /BRCosmetics/Characters/Player/Male/Medium/Bodies/M_MED_EvilBunny/Meshes/Parts/M_MED_Evil_Bunny_Head_AnimBP
target_skeleton=&/M_MED_Evil_Bunny_Head_Skeleton
graphs [
 AnimGraph {nodes [
   n1 = Get trail_alpha() -> (trail_alpha:double)
   n2 = Get look_down_alpha() -> (look_down_alpha:double)
   n3 = "Transform (Modify) Bone - Bone: dyn_ear_01_r"(
    ComponentPose=Local To Component(LocalPose=Use cached pose 'copy_pose'())
    Alpha=@n2
    Translation="0,0,0"
    Rotation="50,-20,0"
    Scale="1,1,1"
   ) {
    rotationMode=BMM_Additive
    rotationSpace=BCS_BoneSpace
    alphaBoolBlend {blendOption=Linear}
    boneToModify {boneName=dyn_ear_01_r}
    rotation {pitch=50, yaw=-20}
   }
   n4 = "Transform (Modify) Bone - Bone: dyn_ear_01_l"(ComponentPose=@n3, Alpha=@n2, Translation="0,0,0", Rotation="50,20,0", Scale="1,1,1") {
    rotationMode=BMM_Additive
    rotationSpace=BCS_BoneSpace
    alphaBoolBlend {blendOption=Linear}
    boneToModify {boneName=dyn_ear_01_l}
    rotation {pitch=50, yaw=20}
   }
   n5 = Trail controller - Bone: dyn_ear_04_r!(ComponentPose=@n4, Alpha=@n1, FakeVelocity:Vector="0,0,0") -> (Pose) {
    chainLength=4
    relaxationSpeedScale=2
    alphaScaleBiasClamp {bInterpResult=true}
    rotationLimits [{limitMax=<180 180 180>}, {limitMax=<180 180 180>}, {limitMax=<180 180 180>}, {limitMax=<180 180 180>}]
    rotationOffsets [<0 0 0>, <0 0 0>, <0 0 0>, <0 0 0>]
    trailBone {boneName=dyn_ear_04_r}
   }
   n6 = Spring controller - Bone: dyn_ear_02_r!(ComponentPose:Component Space Pose Link Structure=@n5, Alpha:float=1) -> (Pose:Component Space Pose Link Structure) {
    is=title AnimGraphNode_SpringBone
    bLimitDisplacement=true
    bRotateX=true
    bRotateY=true
    bRotateZ=true
    bTranslateX=false
    alphaBoolBlend {blendOption=Linear}
    springBone {boneName=dyn_ear_02_r}
   }
   n7 = Trail controller - Bone: dyn_ear_04_l!(ComponentPose=@n6, Alpha=@n1, FakeVelocity:Vector="0,0,0") -> (Pose) {
    chainLength=4
    relaxationSpeedScale=2
    alphaScaleBiasClamp {bInterpResult=true}
    rotationLimits [{limitMax=<180 180 180>}, {limitMax=<180 180 180>}, {limitMax=<180 180 180>}, {limitMax=<180 180 180>}]
    rotationOffsets [<0 0 0>, <0 0 0>, <0 0 0>, <0 0 0>]
    trailBone {boneName=dyn_ear_04_l}
   }
   n8 = Spring controller - Bone: dyn_ear_02_l!(ComponentPose:Component Space Pose Link Structure=@n7, Alpha:float=1) -> (Pose:Component Space Pose Link Structure) {
    is=title AnimGraphNode_SpringBone
    bLimitDisplacement=true
    bRotateX=true
    bRotateY=true
    bRotateZ=true
    bTranslateX=false
    alphaBoolBlend {blendOption=Linear}
    springBone {boneName=dyn_ear_02_l}
   }
   n9 = Anim Dynamics - Bone: dyn_ear_01_l(ComponentPose=Local To Component(LocalPose=Use cached pose 'copy_pose'()), Alpha=1, ExternalForce="0,0,0") {
    angularDampingOverride=1
    angularSpringConstant=1500
    bAngularSpring=true
    bGravityOverrideInSimSpace=true
    bOverrideAngularDamping=true
    gravityScale=0.35
    simulationSpace=RootRelative
    alphaBoolBlend {blendOption=Linear}
    boundBone {boneName=dyn_ear_01_l}
    physicsBodyDefinitions [{
      sphereCollisionRadius=0
      boundBone {boneName=dyn_ear_01_l}
      boxExtents=<3 10 10>
      constraintSetup {angularLimitsMax=<0 0 10>, angularLimitsMin=<0 -40 -10>, angularTarget=<0 0 0>}
      localJointOffset=<3 0 -2>
     }]
   }
   n10 = Anim Dynamics - Bone: dyn_ear_01_r(ComponentPose=@n9, Alpha=Get is_targeting_alpha() -> (is_targeting_alpha:double), ExternalForce="0,0,0") {
    angularDampingOverride=1
    angularSpringConstant=1500
    bAngularSpring=true
    bGravityOverrideInSimSpace=true
    bOverrideAngularDamping=true
    gravityScale=0.35
    simulationSpace=RootRelative
    alphaBoolBlend {blendOption=Linear}
    boundBone {boneName=dyn_ear_01_r}
    physicsBodyDefinitions [{
      sphereCollisionRadius=0
      boundBone {boneName=dyn_ear_01_r}
      boxExtents=<3 10 10>
      constraintSetup {angularLimitsMax=<0 0 10>, angularLimitsMin=<0 -40 -10>, angularTarget=<0 0 0>}
      localJointOffset=<3 0 -2>
     }]
   }
   n11 = RigidBody(ComponentPose=@n10) {
    bClampLinearTranslationLimitToRefPose=true
    overridePhysicsAsset=&/M_MED_Evil_Bunny_Head_Physics
    simulationSpace=BaseBoneSpace
    alphaBoolBlend {blendOption=Linear}
    baseBoneRef {boneName=root}
    componentLinearAccScale=<1 1 1>
    componentLinearVelScale=<1 1 1>
   }
   Output Pose(Result=Blend Poses by bool(
     BlendPose_0=Component To Local(ComponentPose=@n8)
     BlendPose_1=Component To Local(ComponentPose=@n11)
     bActiveValue=Get physics_disabled() -> (physics_disabled:bool)
    ) {blendType=Linear}) {layerGroup=None}
   Save cached pose 'copy_pose'(Pose=Copy Pose From Mesh(SourceMeshComponent=Get MeshToCopy() -> (MeshToCopy:Skeletal Mesh Component Object Reference)) {bUseAttachedParent=true}) {cacheName=copy_pose}
  ]}
 EventGraph {nodes [
   Event BlueprintUpdateAnimation()
   > n1 = Sequence!(execute) -> (then_0, then_1, then_2:Exec)
   n2 = Get player_pawn() -> (player_pawn:Fort Player Pawn Athena Object Reference)
   > Is Valid(InputObject=@n2)
   n3 = Get Actor Rotation!(self:Actor Object Reference=@n2) -> (ReturnValue:Rotator) {is=title K2Node_CallFunction}
   n4 = BreakVector(InVec=vector * vector(
     A:Vector=Get Rotation X Vector!(InRot:Rotator=@n3) -> (ReturnValue:Vector) {is=title K2Node_CallFunction}
     B:Vector=GetVelocity!(self:Actor Object Reference=@n2) -> (ReturnValue:Vector)
    ) -> (ReturnValue:Vector) {is=title K2Node_CommutativeAssociativeBinaryOperator})
   n5 = float + float(A=@n4, B=@n4.Y) {is=title K2Node_CommutativeAssociativeBinaryOperator}
   > Set pawn_speed(pawn_speed:double=@n5) -> (Output_Get:double)
   > Set isMovingBackward(isMovingBackward:bool="float < float"!(A:double=@n5, B:double=0) -> (ReturnValue:bool) {is=title K2Node_CallFunction}) -> (Output_Get:bool)
   > Set trail_alpha(trail_alpha:double=MapRangeClamped(Value=Get pawn_speed() -> (pawn_speed:double), InRangeA=0, InRangeB=200, OutRangeA=1, OutRangeB=0)) -> (Output_Get:double)
   n6 = BreakTransform(InTransform=GetDeltaTransformFromRefPose(self=GetPartSkeletalMeshComponent(), BoneName=head, BaseName=spine_05))
   Set look_down_alpha(execute=@n1.then_1, look_down_alpha:double=MapRangeClamped(Value=BreakRotator(InRot=@n6.Rotation).Yaw, InRangeA=-10, InRangeB=25)) -> (Output_Get:double)
   n7 = Get CurrentLOD()
   n8 = Get bRigidBodyEnabled()
   Set physics_disabled(execute=@n1.then_2, physics_disabled:bool=OR Boolean(A=NOT Boolean(A=@n8), B="integer > integer"(A=@n7, B=1) {is=title K2Node_CallFunction})) -> (Output_Get:bool)
   > Set is_targeting_alpha(is_targeting_alpha:double=SelectFloat(A=0, B=1, bPickA=Get bIsTargeting())) -> (Output_Get:double)
   Event BlueprintInitializeAnimation()
   > n9 = Cast To FortPlayerPawnAthena(Object=GetOwner(self=GetOwnerSkeletalMeshComponent()))
   > Set player_pawn(player_pawn:Fort Player Pawn Athena Object Reference=@n9) -> (Output_Get:Fort Player Pawn Athena Object Reference)
   > Set MeshToCopy(MeshToCopy:Skeletal Mesh Component Object Reference=GetOwnerSkeletalMeshComponent()) -> (Output_Get:Skeletal Mesh Component Object Reference)
   TryGetPawnOwner()
  ]}
]
variables [
 is_targeting_alpha {type=number}
 isMovingBackward {type=boolean}
 look_down_alpha {type=number}
 MeshToCopy {type=SkeletalMeshComponent}
 pawn_speed {type=number}
 physics_disabled {type=boolean}
 player_pawn {type=FortPlayerPawnAthena}
 trail_alpha {type=number}
]
